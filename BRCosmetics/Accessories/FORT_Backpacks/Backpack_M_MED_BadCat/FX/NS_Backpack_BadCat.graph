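# Niagara effect for the BadCat backpack cosmetic, adding fire, smoke, glows and small lights at skeletal mesh sockets.
# Smoke and fire emitters spawn along socket orientation, with system Timelines driving rate, size, speed, lift and extra probability-based spawning.
# Persistent paired Decal, Glow and LightAndGlow particles re-sample their sockets each update.
# A cycling curve indexed by FlickerSpeed flickers the decals, and a front-end modifier dims decals and glows outside gameplay.
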NiagaraSystem /BRCosmetics/Accessories/FORT_Backpacks/Backpack_M_MED_BadCat/FX/NS_Backpack_BadCat
dependencies {
 usedDataInterfaces [
  &NiagaraDataInterfaceMeshRendererInfo
  &NiagaraDataInterfaceCurve
  &NiagaraDataInterfaceArrayFloat3
  &NiagaraDataInterfaceArrayFloat
  &NiagaraDataInterfaceColorCurve
  &NiagaraDataInterfaceSkeletalMesh
 ]
 usedDynamicInputs [&Boolean_OrOperation]
 usedModules [
  &SystemState
  &Timeline
  &self:SystemSpawnScript.SystemScriptSource.SystemScriptGraph.NiagaraNodeAssignment_4.SetVariables_#1
  &self:SystemSpawnScript.SystemScriptSource.SystemScriptGraph.NiagaraNodeAssignment_5.SetVariables_#2
 ]
}
emitters {
 Decal {
  emitterUpdateScript [
   &EmitterState(
    Inactive Response=::NewEnumerator1
    Life Cycle Mode=::NewEnumerator1
    Loop Behavior=::NewEnumerator1
    Loop Duration Mode=::NewEnumerator1
    UseLoopDelay=false
   )
   &SpawnBurst_Instantaneous(Spawn Count=2, Use Spawn Probability=false, UseLoopCountLimit=false)
  ]
  particleSpawnScript [
   &InitializeParticle(
    Color=<rgba 10 1.381 0 1>
    Sprite Rotation Mode=::NewEnumerator2
    Sprite Size Mode=::NewEnumerator3
    Uniform Sprite Size=&SelectFloatFromArray(
     Direct Array Index=&ReturnExecIndex()
     Float Selection Array=DataInterface() {propertyValues {FloatData [4.5,4]}}
    )
    UsePositionOffset=false
    Write Material Random=true
   )
   &SkeletalMeshLocation(
    Bone / Socket Index=&ReturnExecIndex()
    MaskSpawnGroup=false
    Mesh Sampling Type=::NewEnumerator5
    Sampled Position Offset=<3 0 0>
    Skeletal Mesh=$System.Skeletal Mesh
    Socket Sampling Mode=::NewEnumerator5
   )
   &CameraOffset(Camera Offset Amount=3)
  ]
  particleUpdateScript [
   &ParticleState(Kill Particles When Lifetime Has Elapsed=false, Let Infinitely Lived Particles Die When Emitter Deactivates=true)
   &SkeletalMeshLocation(
    Bone / Socket Index=&ReturnExecIndex()
    MaskSpawnGroup=false
    Mesh Sampling Type=::NewEnumerator5
    Sampled Position Offset=&SelectVectorFromArray(
     Direct Array Index=&ReturnExecIndex()
     Vector Selection Array=DataInterface() {propertyValues {FloatData [<2.2 0 0>, <2.2 0 -1>]}}
    )
    Skeletal Mesh=$System.Skeletal Mesh
    Socket Sampling Mode=::NewEnumerator5
   ) {module=SkeletalMeshLocation001}
   &DynamicMaterialParameters(
    Index 0 Param 1=2
    Index 0 Param 2=1
    Index 0 Param 3=&RandomRangeFloat(Minimum=0, Maximum=1)
    Index 0 Param 4=&RandomRangeFloat(Minimum=0, Maximum=1)
    Module.Index0Float0Write=true
    Module.Index0Float1Write=true
    Module.Index0Float2Write=true
    Module.Index0Float3Write=true
   ) {module=DynamicMaterialParameters002}
   &ScaleColor(
    Scale Alpha=&Multiply_Float(
     A=&RemapRange(
      Input Value To Remap=&Multiply_Float(
       A=&FloatFromCurve(
        FloatCurve=DataInterface() {propertyValues {
          bUseLUT=false
          Curve {
           postInfinityExtrap=RCCE_Cycle
           preInfinityExtrap=RCCE_Cycle
           keys [
            <at 0 1 RCIM_Cubic>
            {arriveTangent=0.5283, interpMode=RCIM_Cubic, leaveTangent=0.5283, time=0.0746, value=0.875}
            {arriveTangent=-0.0227, interpMode=RCIM_Cubic, leaveTangent=-0.0227, tangentMode=RCTM_User, time=0.1682, value=1.0888}
            {arriveTangent=-1.3938, interpMode=RCIM_Cubic, leaveTangent=-1.3938, time=0.2019, value=0.6577}
            {arriveTangent=1.4729, interpMode=RCIM_Cubic, leaveTangent=1.4729, time=0.2748, value=0.9402}
            {arriveTangent=0.905, interpMode=RCIM_Cubic, leaveTangent=0.905, time=0.3464, value=0.8705}
            {arriveTangent=-1.2159, interpMode=RCIM_Cubic, leaveTangent=-1.2159, time=0.4074, value=1.0603}
            {arriveTangent=-1.5422, interpMode=RCIM_Cubic, leaveTangent=-1.5422, time=0.4365, value=0.7609}
            {arriveTangent=-0.8529, interpMode=RCIM_Cubic, leaveTangent=-0.8529, time=0.4871, value=0.9373}
            {arriveTangent=0.3027, interpMode=RCIM_Cubic, leaveTangent=0.3027, time=0.5331, value=0.6785}
            {arriveTangent=1.4034, interpMode=RCIM_Cubic, leaveTangent=1.4034, time=0.6112, value=0.9748}
            {arriveTangent=1.5913, interpMode=RCIM_Cubic, leaveTangent=1.5913, time=0.69, value=0.8986}
            {arriveTangent=-0.751, interpMode=RCIM_Cubic, leaveTangent=-0.751, time=0.7536, value=1.2014}
            {arriveTangent=-2.595, interpMode=RCIM_Cubic, leaveTangent=-2.595, time=0.7995, value=0.8164}
            {arriveTangent=-0.8671, interpMode=RCIM_Cubic, leaveTangent=-0.8671, time=0.8674, value=0.9059}
            {arriveTangent=-0.3671, interpMode=RCIM_Cubic, leaveTangent=-0.3671, time=0.9138, value=0.7173}
            {arriveTangent=-0.6714, interpMode=RCIM_Cubic, leaveTangent=-0.6714, time=0.9899, value=0.8609}
            {arriveTangent=0.7195, interpMode=RCIM_Cubic, leaveTangent=0.7195, time=1.0372, value=0.6345}
            {arriveTangent=1.5649, interpMode=RCIM_Cubic, leaveTangent=1.5649, time=1.099, value=0.9394}
            {arriveTangent=0.3682, interpMode=RCIM_Cubic, leaveTangent=0.3682, time=1.1296, value=0.7792}
            {arriveTangent=1.199, interpMode=RCIM_Cubic, leaveTangent=1.199, time=1.2078, value=0.9795}
            {arriveTangent=0.6999, interpMode=RCIM_Cubic, leaveTangent=0.6999, time=1.2281, value=0.8972}
            {arriveTangent=0.8013, interpMode=RCIM_Cubic, leaveTangent=0.8013, time=1.329, value=1.0643}
            {arriveTangent=0.4649, interpMode=RCIM_Cubic, leaveTangent=0.4649, time=1.3719, value=1.0125}
            {arriveTangent=-3.0415, interpMode=RCIM_Cubic, leaveTangent=-3.0415, time=1.414, value=1.1038}
            {arriveTangent=-0.0928, interpMode=RCIM_Cubic, leaveTangent=-0.0928, tangentMode=RCTM_User, time=1.4501, value=0.7746}
            {arriveTangent=0.7682, interpMode=RCIM_Cubic, leaveTangent=0.7682, time=1.5242, value=0.9899}
            {arriveTangent=0.5284, interpMode=RCIM_Cubic, leaveTangent=0.5284, time=1.5566, value=0.8565}
            {arriveTangent=-0.1056, interpMode=RCIM_Cubic, leaveTangent=-0.1056, time=1.7244, value=1.0957}
            {arriveTangent=-2.1193, interpMode=RCIM_Cubic, leaveTangent=-2.1193, time=1.7619, value=0.8348}
            {arriveTangent=-1.2854, interpMode=RCIM_Cubic, leaveTangent=-1.2854, time=1.7959, value=0.9442}
            {arriveTangent=-0.1279, interpMode=RCIM_Cubic, leaveTangent=-0.1279, time=1.836, value=0.7395}
            {arriveTangent=1.0187, interpMode=RCIM_Cubic, leaveTangent=1.0187, time=1.9343, value=0.9265}
            {arriveTangent=1.0782, interpMode=RCIM_Cubic, leaveTangent=1.0782, time=1.9912, value=0.8977}
            {arriveTangent=-2.5847, interpMode=RCIM_Cubic, leaveTangent=-2.5847, time=2.0488, value=1.0499}
            {arriveTangent=-0.6379, interpMode=RCIM_Cubic, leaveTangent=-0.6379, time=2.0828, value=0.6611}
            {arriveTangent=0.4353, interpMode=RCIM_Cubic, leaveTangent=0.4353, time=2.2158, value=0.9434}
            {arriveTangent=-0.3447, interpMode=RCIM_Cubic, leaveTangent=-0.3447, time=2.2673, value=0.7414}
            {arriveTangent=0.6603, interpMode=RCIM_Cubic, leaveTangent=0.6603, time=2.3258, value=0.9055}
            {arriveTangent=0.3801, interpMode=RCIM_Cubic, leaveTangent=0.3801, time=2.3971, value=0.8271}
            {arriveTangent=0.6687, interpMode=RCIM_Cubic, leaveTangent=0.6687, time=2.4654, value=0.9586}
            {arriveTangent=1.2527, interpMode=RCIM_Cubic, leaveTangent=1.2527, time=2.5143, value=0.9055}
            {arriveTangent=-0.515, interpMode=RCIM_Cubic, leaveTangent=-0.515, time=2.5566, value=1.0728}
            {arriveTangent=-1.6201, interpMode=RCIM_Cubic, leaveTangent=-1.6201, time=2.597, value=0.8629}
            {arriveTangent=-0.9669, interpMode=RCIM_Cubic, leaveTangent=-0.9669, time=2.6456, value=0.9286}
            {arriveTangent=0.1009, interpMode=RCIM_Cubic, leaveTangent=0.1009, time=2.6921, value=0.771}
            {arriveTangent=0.0929, interpMode=RCIM_Cubic, leaveTangent=0.0929, time=2.7509, value=0.9392}
            {arriveTangent=0.2939, interpMode=RCIM_Cubic, leaveTangent=0.2939, time=2.8654, value=0.7871}
            {arriveTangent=1.7227, interpMode=RCIM_Cubic, leaveTangent=1.7227, time=2.9271, value=0.991}
            {arriveTangent=0.1236, interpMode=RCIM_Cubic, leaveTangent=0.1236, time=2.957, value=0.9449}
            <at 3 1 RCIM_Cubic>
           ]
          }
         }}
        CurveIndex=$System.FlickerSpeed
       )
       B=1
      )
      Input Min=0.6
      Input Max=1.2
      Output Min=0
      Output Max=1
     )
     B=&Multiply_Float(
      A=&Lerp_Float(A=0.5, B=1.5, Alpha=$System.Timeline005.NormalizedTime)
      B=&FN_FrontEndFloatModifier(Gameplay=1, Front End=0.5)
     )
    )
    ScaleA=true
    ScaleRGB=true
   )
  ]
  rendererClasses [&NiagaraSpriteRendererProperties]
  renderers [
   {rendererClass=&NiagaraSpriteRendererProperties}
   {rendererClass=&NiagaraSpriteRendererProperties, rendererIndex=1}
  ]
 }
 FN_Core_Fiery_Smoke_Sprite {
  emitterUpdateScript [
   &EmitterState()
   &SpawnRate(
    SpawnRate=&Lerp_Float(
     A=20
     B=50
     Alpha=&MakeCustomFloatFromBool(Bool=&SetBoolByFloatComparison(A=$System.Timeline002.NormalizedTime, B=0), True Float=1, False Float=0)
    )
    Use Spawn Probability=false
   )
  ]
  particleSpawnScript [
   &PartitionParticles(Partition Mode:ENiagara_PartitionMode=::NewEnumerator5)
   &InitializeParticle(
    Color Maximum=<rgba 0.1255 0.1137 0.1059 1>
    Color Minimum=<rgba 0.2 0.1843 0.1686 1>
    Color Mode=::NewEnumerator2
    Lifetime Max=1
    Lifetime Min=0.5
    Lifetime Mode=::NewEnumerator1
    Sprite Rotation Angle=&RandomRangeFloat(Minimum=0, Maximum=360)
    Sprite Rotation Mode=::NewEnumerator0
    Sprite Size Mode=::NewEnumerator3
    Uniform Sprite Size=&Multiply_Float(
     A=&Multiply_Float(
      A=&RandomRangeFloat(Minimum=15, Maximum=20)
      B=&Lerp_Float(A=1.1, B=1.3, Alpha=$System.Timeline002.NormalizedTime)
     )
     B=&MaskFloat(Mask Parameter=$Particles.PartitionParticles.Partition, Mask=0, Masked Float=1, Passthrough Float=0.8)
    )
    UsePositionOffset=false
    Write Material Random=true
   )
   /SetVariables(Particles.InitSpriteRotRate=&UniformRangedFloat(Minimum=-3, Maximum=3)) {isSetParameters=true, module=SetVariables_#3}
   &SkeletalMeshLocation(
    Bone / Socket Index=$Particles.PartitionParticles.Partition
    MaskSpawnGroup=false
    Mesh Sampling Type=::NewEnumerator5
    Skeletal Mesh=$System.Skeletal Mesh
    Socket Sampling Mode=::NewEnumerator5
   )
   &AddVelocity(
    Rotation Mode=::NewEnumerator5
    Velocity=&RotateVectorByQuaternion(Vector To Rotate=<1 0 0>, Quaternion=$Particles.SkeletalMeshLocation.SampledRotation)
    Velocity Speed Scale=&Lerp_Float(A=20, B=40, Alpha=$System.Timeline002.NormalizedTime)
   )
  ]
  particleUpdateScript [
   &ParticleState()
   &SkeletalMeshLocation(
    Bone / Socket Index=$Particles.PartitionParticles.Partition
    Initialize Particle Position To Sampled Position=false
    MaskSpawnGroup=false
    Mesh Sampling Type=::NewEnumerator5
    Position Sampling=::NewEnumerator1
    Skeletal Mesh=$System.Skeletal Mesh
    Socket Sampling Mode=::NewEnumerator5
   ) {module=SkeletalMeshLocation001}
   &InheritSourceMovement(ApplyForce=false, ApplyPosition=true, ApplyPositionFalloff=false) {enabled=false}
   &SpriteRotationRate(Rotation Rate=&Multiply_Float(
     A=$Particles.InitSpriteRotRate
     B=&FloatFromCurve(
      FloatCurve=DataInterface() {propertyValues {Curve {keys [
          {arriveTangent=-4.0107, interpMode=RCIM_Cubic, leaveTangent=-4.0107, tangentMode=RCTM_User, time=-0.0046, value=1.0113}
          {arriveTangent=-0.2496, interpMode=RCIM_Cubic, leaveTangent=-0.2496, tangentMode=RCTM_User, time=0.5, value=0.16}
         ]}}}
      CurveIndex=$Emitter.Age
      Scale Curve=150
     )
    ))
   &ScaleColor(
    Scale Alpha=&FloatFromCurve(
     FloatCurve=DataInterface() {propertyValues {Curve {keys [{interpMode=RCIM_Cubic, tangentMode=RCTM_User, time=0, value=1}, {interpMode=RCIM_Cubic, tangentMode=RCTM_User, time=1, value=0}]}}}
     Scale Curve=&Lerp_Float(A=1, B=1.5, Alpha=$System.Timeline002.NormalizedTime)
    )
    Scale RGB=&VectorFromFloat(Value=&FloatFromCurve(FloatCurve=DataInterface() {propertyValues {Curve {keys [{arriveTangent=1.8145, interpMode=RCIM_Cubic, leaveTangent=4.5795, tangentMode=RCTM_Break, time=0, value=0.4}, <at 0.3387 0.9565>, <at 1 1>]}}}))
    ScaleA=true
    ScaleRGB=true
   )
   &ScaleSpriteSize(Scale Factor=&Vector2DFromFloat(Value=&FloatFromCurve(FloatCurve=DataInterface() {propertyValues {Curve {keys [
          {arriveTangent=-0.0465, interpMode=RCIM_Cubic, leaveTangent=-0.0465, tangentMode=RCTM_User, time=0, value=0.15}
          {arriveTangent=0.0148, interpMode=RCIM_Cubic, leaveTangent=0.0148, tangentMode=RCTM_User, time=1.0463, value=0.8}
         ]}}})))
   &DynamicMaterialParameters(
    Index 0 Param 1=&Add_Float(
     A=$Particles.RandomFloat
     B=&FloatFromCurve(FloatCurve=DataInterface() {propertyValues {Curve {keys [
          {arriveTangent=2.7533, interpMode=RCIM_Cubic, leaveTangent=2.847, tangentMode=RCTM_Break, time=0, value=0}
          {arriveTangent=1.3581, interpMode=RCIM_Cubic, leaveTangent=1.3581, tangentMode=RCTM_User, time=0.2558, value=0.5125}
          {arriveTangent=0.0529, interpMode=RCIM_Cubic, leaveTangent=0.2021, tangentMode=RCTM_User, time=1, value=1}
         ]}}})
    )
    Index 0 Param 3=&FloatFromCurve(
     FloatCurve=DataInterface() {propertyValues {Curve {keys [
         {arriveTangent=-0.3554, interpMode=RCIM_Cubic, leaveTangent=-0.3554, tangentMode=RCTM_User, time=0, value=1}
         {arriveTangent=-0.0712, interpMode=RCIM_Cubic, leaveTangent=0.2723, tangentMode=RCTM_Break, time=0.4, value=0}
        ]}}}
     Scale Curve=0.15
    )
    Index 0 Param 4=&FloatFromCurve(FloatCurve=DataInterface() {propertyValues {Curve {keys [<at 0.5 1>,<at 1 0>]}}})
    Index 1 Param 1=&FloatFromCurve(
     FloatCurve=DataInterface() {propertyValues {Curve {keys [
         {interpMode=RCIM_Cubic, leaveTangent=4.7088, tangentMode=RCTM_Break, time=0, value=0}
         {arriveTangent=4.2257, interpMode=RCIM_Cubic, leaveTangent=4.2257, tangentMode=RCTM_User, time=0.25, value=1}
        ]}}}
     Scale Curve=1.5
    )
    Index 1 Param 2=&FloatFromCurve(
     FloatCurve=DataInterface() {propertyValues {Curve {keys [<at 0 0.2>, {leaveTangent=0.9066, time=1, value=2}]}}}
     Scale Curve=&RandomRangeFloat(Minimum=0.5, Maximum=0.6)
    )
    Index 1 Param 3=&FloatFromCurve(FloatCurve=DataInterface() {propertyValues {Curve {keys [
         {interpMode=RCIM_Cubic, leaveTangent=6.3917, tangentMode=RCTM_Break, time=0, value=0.2}
         {arriveTangent=4.8433, interpMode=RCIM_Cubic, leaveTangent=19.0929, tangentMode=RCTM_User, time=0.25, value=1}
        ]}}})
    Index 1 Param 4=&FloatFromCurve(
     FloatCurve=DataInterface() {propertyValues {Curve {keys [{arriveTangent=-0.3224, interpMode=RCIM_Cubic, leaveTangent=-0.3224, tangentMode=RCTM_User, time=0, value=1}, <at 1 0.9>]}}}
     Scale Curve=&RandomRangeFloat(Minimum=1.1, Maximum=1.4)
    )
    Index 2 Param 1=&FloatFromCurve(FloatCurve=DataInterface() {propertyValues {Curve {keys [
         {arriveTangent=2.646, interpMode=RCIM_Cubic, leaveTangent=2.646, tangentMode=RCTM_User, time=0, value=0}
         {leaveTangent=-0.0719, time=1, value=1}
        ]}}})
    Index 2 Param 2=&FloatFromCurve(
     FloatCurve=DataInterface() {propertyValues {Curve {keys [
         {arriveTangent=1.9056, time=0, value=0.2}
         {arriveTangent=0.0259, interpMode=RCIM_Cubic, leaveTangent=0.0259, tangentMode=RCTM_User, time=1, value=1.4}
        ]}}}
     Scale Curve=&RandomRangeFloat(Minimum=0.5, Maximum=1)
    )
    Index 2 Param 3=&FloatFromCurve(FloatCurve=DataInterface() {propertyValues {Curve {keys [
         {interpMode=RCIM_Cubic, leaveTangent=2.3064, tangentMode=RCTM_Break, time=0, value=1}
         {interpMode=RCIM_Cubic, leaveTangent=11.1233, tangentMode=RCTM_Break, time=1, value=2}
        ]}}})
    Index 2 Param 4=&FloatFromCurve(
     FloatCurve=DataInterface() {propertyValues {Curve {keys [
         {arriveTangent=-0.6867, interpMode=RCIM_Cubic, leaveTangent=-0.6867, tangentMode=RCTM_User, time=0, value=1}
         {interpMode=RCIM_Cubic, leaveTangent=-1.2956, tangentMode=RCTM_Break, time=1, value=0.75}
        ]}}}
     Scale Curve=1.6
    )
    Module.Param0WriteEnabled=true
    Module.Param0WriteEnabled1=true
    Module.Param0WriteEnabled2=true
    Module.Param1WriteEnabled=false
    Module.Param1WriteEnabled1=true
    Module.Param1WriteEnabled2=true
    Module.Param2WriteEnabled=true
    Module.Param2WriteEnabled1=true
    Module.Param2WriteEnabled2=true
    Module.Param3WriteEnabled=true
    Module.Param3WriteEnabled1=true
    Module.Param3WriteEnabled2=true
    Write Parameter Index 1=true
    Write Parameter Index 2=true
   )
   &AccelerationForce(Acceleration=&MakeVector(
     X=0
     Y=0
     Z=&Lerp_Float(
      A=&MaskFloat(Mask Parameter=$Particles.Initial.PartitionParticles.Partition, Mask=0, Masked Float=1, Passthrough Float=0.5)
      B=120
      Alpha=$System.Timeline002.NormalizedTime
     )
    ))
   &Drag(Drag=3, Use Linear Drag=true, Use Rotational Drag=false)
   &SolveForcesAndVelocity(Clamp Velocity=false, Limit Acceleration=false)
  ]
  rendererClasses [&NiagaraSpriteRendererProperties]
  renderers [
   {rendererClass=&NiagaraSpriteRendererProperties}
   {rendererClass=&NiagaraSpriteRendererProperties, rendererIndex=1}
  ]
 }
 FN_Core_Fiery_Smoke_Sprite001 {
  emitterUpdateScript [
   &EmitterState()
   &SpawnRate(Spawn Probability=$System.Timeline002.Float, SpawnRate=40, Use Spawn Probability=true) {module=SpawnRate001}
   &SpawnRate(SpawnRate=&Lerp_Float(A=20, B=50, Alpha=$System.Timeline002.NormalizedTime), Use Spawn Probability=false) {enabled=false}
  ]
  particleSpawnScript [
   &InitializeParticle(
    Color Maximum=<rgba 0.1255 0.1137 0.1059 1>
    Color Minimum=<rgba 0.2 0.1843 0.1686 1>
    Color Mode=::NewEnumerator2
    Lifetime Max=0.65
    Lifetime Min=0.5
    Lifetime Mode=::NewEnumerator1
    Sprite Rotation Angle=&RandomRangeFloat(Minimum=0, Maximum=360)
    Sprite Rotation Mode=::NewEnumerator0
    Sprite Size Mode=::NewEnumerator3
    Uniform Sprite Size=&Multiply_Float(
     A=&RandomRangeFloat(Minimum=15, Maximum=20)
     B=&Lerp_Float(A=1.1, B=2, Alpha=$System.Timeline002.NormalizedTime)
    )
    UsePositionOffset=false
    Write Material Random=true
   )
   &PartitionParticles(Partition Mode:ENiagara_PartitionMode=::NewEnumerator5)
   /SetVariables(Particles.InitSpriteRotRate=&UniformRangedFloat(Minimum=-3, Maximum=3)) {isSetParameters=true, module=SetVariables_#3}
   &SkeletalMeshLocation(
    Bone / Socket Index=$Particles.PartitionParticles.Partition
    MaskSpawnGroup=false
    Mesh Sampling Type=::NewEnumerator5
    Skeletal Mesh=$System.Skeletal Mesh
    Socket Sampling Mode=::NewEnumerator5
   )
   &AddVelocity(
    Rotation Mode=::NewEnumerator5
    Velocity=&RotateVectorByQuaternion(Vector To Rotate=<1 0 0>, Quaternion=$Particles.SkeletalMeshLocation.SampledRotation)
    Velocity Speed Scale=&Lerp_Float(A=25, B=40, Alpha=$System.Timeline002.NormalizedTime)
   )
  ]
  particleUpdateScript [
   &ParticleState()
   &SkeletalMeshLocation(
    Bone / Socket Index=$Particles.PartitionParticles.Partition
    Initialize Particle Position To Sampled Position=false
    MaskSpawnGroup=false
    Mesh Sampling Type=::NewEnumerator5
    Position Sampling=::NewEnumerator1
    Skeletal Mesh=$System.Skeletal Mesh
    Socket Sampling Mode=::NewEnumerator5
   ) {module=SkeletalMeshLocation001}
   &InheritSourceMovement(ApplyForce=false, ApplyPosition=true, ApplyPositionFalloff=false) {enabled=false}
   &SpriteRotationRate(Rotation Rate=&Multiply_Float(
     A=$Particles.InitSpriteRotRate
     B=&FloatFromCurve(
      FloatCurve=DataInterface() {propertyValues {Curve {keys [
          {arriveTangent=-4.0107, interpMode=RCIM_Cubic, leaveTangent=-4.0107, tangentMode=RCTM_User, time=-0.0046, value=1.0113}
          {arriveTangent=-0.2496, interpMode=RCIM_Cubic, leaveTangent=-0.2496, tangentMode=RCTM_User, time=0.5, value=0.16}
         ]}}}
      CurveIndex=$Emitter.Age
      Scale Curve=150
     )
    ))
   &ScaleColor(
    Scale Alpha=&FloatFromCurve(
     FloatCurve=DataInterface() {propertyValues {Curve {keys [{interpMode=RCIM_Cubic, tangentMode=RCTM_User, time=0, value=1}, {interpMode=RCIM_Cubic, tangentMode=RCTM_User, time=1, value=0}]}}}
     Scale Curve=5
    )
    Scale RGB=&VectorFromFloat(Value=&FloatFromCurve(FloatCurve=DataInterface() {propertyValues {Curve {keys [{arriveTangent=1.8145, interpMode=RCIM_Cubic, leaveTangent=4.5795, tangentMode=RCTM_Break, time=0, value=0.4}, <at 0.3387 0.9565>, <at 1 1>]}}}))
    ScaleA=true
    ScaleRGB=true
   )
   &ScaleSpriteSize(Scale Factor=&Vector2DFromFloat(Value=&FloatFromCurve(FloatCurve=DataInterface() {propertyValues {Curve {keys [
          {arriveTangent=-0.0465, interpMode=RCIM_Cubic, leaveTangent=-0.0465, tangentMode=RCTM_User, time=0, value=0.15}
          {arriveTangent=0.0148, interpMode=RCIM_Cubic, leaveTangent=0.0148, tangentMode=RCTM_User, time=1.0463, value=0.8}
         ]}}})))
   &DynamicMaterialParameters(
    Index 0 Param 1=&Add_Float(
     A=$Particles.RandomFloat
     B=&FloatFromCurve(FloatCurve=DataInterface() {propertyValues {Curve {keys [
          {arriveTangent=2.7533, interpMode=RCIM_Cubic, leaveTangent=2.847, tangentMode=RCTM_Break, time=0, value=0}
          {arriveTangent=1.3581, interpMode=RCIM_Cubic, leaveTangent=1.3581, tangentMode=RCTM_User, time=0.2558, value=0.5125}
          {arriveTangent=0.0529, interpMode=RCIM_Cubic, leaveTangent=0.2021, tangentMode=RCTM_User, time=1, value=1}
         ]}}})
    )
    Index 0 Param 3=&FloatFromCurve(
     FloatCurve=DataInterface() {propertyValues {Curve {keys [
         {arriveTangent=-0.3554, interpMode=RCIM_Cubic, leaveTangent=-0.3554, tangentMode=RCTM_User, time=0, value=1}
         {arriveTangent=-0.0712, interpMode=RCIM_Cubic, leaveTangent=0.2723, tangentMode=RCTM_Break, time=0.4, value=0}
        ]}}}
     Scale Curve=0.15
    )
    Index 0 Param 4=&FloatFromCurve(FloatCurve=DataInterface() {propertyValues {Curve {keys [<at 0.5 1>,<at 1 0>]}}})
    Index 1 Param 1=&FloatFromCurve(
     FloatCurve=DataInterface() {propertyValues {Curve {keys [
         {interpMode=RCIM_Cubic, leaveTangent=4.7088, tangentMode=RCTM_Break, time=0, value=0}
         {arriveTangent=4.2257, interpMode=RCIM_Cubic, leaveTangent=4.2257, tangentMode=RCTM_User, time=0.25, value=1}
        ]}}}
     Scale Curve=1.5
    )
    Index 1 Param 2=&FloatFromCurve(
     FloatCurve=DataInterface() {propertyValues {Curve {keys [<at 0 0.2>, {leaveTangent=0.9066, time=1, value=2}]}}}
     Scale Curve=&RandomRangeFloat(Minimum=0.5, Maximum=0.6)
    )
    Index 1 Param 3=&FloatFromCurve(FloatCurve=DataInterface() {propertyValues {Curve {keys [
         {interpMode=RCIM_Cubic, leaveTangent=6.3917, tangentMode=RCTM_Break, time=0, value=0.2}
         {arriveTangent=4.8433, interpMode=RCIM_Cubic, leaveTangent=19.0929, tangentMode=RCTM_User, time=0.25, value=1}
        ]}}})
    Index 1 Param 4=&FloatFromCurve(
     FloatCurve=DataInterface() {propertyValues {Curve {keys [{arriveTangent=-0.3224, interpMode=RCIM_Cubic, leaveTangent=-0.3224, tangentMode=RCTM_User, time=0, value=1}, <at 1 0.9>]}}}
     Scale Curve=&RandomRangeFloat(Minimum=1.1, Maximum=1.4)
    )
    Index 2 Param 1=&FloatFromCurve(FloatCurve=DataInterface() {propertyValues {Curve {keys [
         {arriveTangent=2.646, interpMode=RCIM_Cubic, leaveTangent=2.646, tangentMode=RCTM_User, time=0, value=0}
         {leaveTangent=-0.0719, time=1, value=1}
        ]}}})
    Index 2 Param 2=&FloatFromCurve(
     FloatCurve=DataInterface() {propertyValues {Curve {keys [
         {arriveTangent=1.9056, time=0, value=0.2}
         {arriveTangent=0.0259, interpMode=RCIM_Cubic, leaveTangent=0.0259, tangentMode=RCTM_User, time=1, value=1.4}
        ]}}}
     Scale Curve=&RandomRangeFloat(Minimum=0.5, Maximum=1)
    )
    Index 2 Param 3=&FloatFromCurve(FloatCurve=DataInterface() {propertyValues {Curve {keys [
         {interpMode=RCIM_Cubic, leaveTangent=2.3064, tangentMode=RCTM_Break, time=0, value=1}
         {interpMode=RCIM_Cubic, leaveTangent=11.1233, tangentMode=RCTM_Break, time=1, value=2}
        ]}}})
    Index 2 Param 4=&FloatFromCurve(
     FloatCurve=DataInterface() {propertyValues {Curve {keys [
         {arriveTangent=-0.6867, interpMode=RCIM_Cubic, leaveTangent=-0.6867, tangentMode=RCTM_User, time=0, value=1}
         {interpMode=RCIM_Cubic, leaveTangent=-1.2956, tangentMode=RCTM_Break, time=1, value=0.75}
        ]}}}
     Scale Curve=1.6
    )
    Module.Param0WriteEnabled=true
    Module.Param0WriteEnabled1=true
    Module.Param0WriteEnabled2=true
    Module.Param1WriteEnabled=false
    Module.Param1WriteEnabled1=true
    Module.Param1WriteEnabled2=true
    Module.Param2WriteEnabled=true
    Module.Param2WriteEnabled1=true
    Module.Param2WriteEnabled2=true
    Module.Param3WriteEnabled=true
    Module.Param3WriteEnabled1=true
    Module.Param3WriteEnabled2=true
    Write Parameter Index 1=true
    Write Parameter Index 2=true
   )
   &AccelerationForce(Acceleration=&MakeVector(X=0, Y=0, Z=&Lerp_Float(A=20, B=120, Alpha=$System.Timeline002.NormalizedTime)))
   &Drag(Drag=3, Use Linear Drag=true, Use Rotational Drag=false)
   &SolveForcesAndVelocity(Clamp Velocity=false, Limit Acceleration=false)
  ]
  rendererClasses [&NiagaraSpriteRendererProperties]
  renderers [
   {rendererClass=&NiagaraSpriteRendererProperties}
   {rendererClass=&NiagaraSpriteRendererProperties, rendererIndex=1}
  ]
 }
 Fire {
  emitterUpdateScript [
   &EmitterState()
   &SpawnRate(SpawnRate=&Lerp_Float(A=20, B=40, Alpha=$System.Timeline002.Time), Use Spawn Probability=false)
  ]
  particleSpawnScript [
   &InitializeParticle(
    Color=<rgba 0.7812 0.4232 0.4232 1>
    Lifetime=0.15
    Sprite Size=<2 10>
    Sprite Size Mode=::NewEnumerator4
    UsePositionOffset=false
    Write Material Random=true
   )
   &PartitionParticles(Partition Mode:ENiagara_PartitionMode=::NewEnumerator5)
   &SkeletalMeshLocation(
    Bone / Socket Index=$Particles.PartitionParticles.Partition
    MaskSpawnGroup=false
    Mesh Sampling Type=::NewEnumerator5
    Sampled Position Offset=<1 0 0>
    Skeletal Mesh=$System.Skeletal Mesh
    Socket Sampling Mode=::NewEnumerator5
   )
   &AddVelocity(
    Rotation Mode=::NewEnumerator5
    Velocity=&RotateVectorByQuaternion(Vector To Rotate=<1 0 0>, Quaternion=$Particles.SkeletalMeshLocation.SampledRotation)
    Velocity Speed Scale=20
   )
  ]
  particleUpdateScript [
   &ParticleState()
   &SkeletalMeshLocation(
    Bone / Socket Index=$Particles.PartitionParticles.Partition
    Initialize Particle Position To Sampled Position=false
    MaskSpawnGroup=false
    Mesh Sampling Type=::NewEnumerator5
    Orientation Sampling=::NewEnumerator1
    Position Sampling=::NewEnumerator1
    Skeletal Mesh=$System.Skeletal Mesh
    Socket Sampling Mode=::NewEnumerator5
   ) {module=SkeletalMeshLocation001}
   &SolveForcesAndVelocity(Clamp Velocity=false, Limit Acceleration=false)
   &DynamicMaterialParameters(
    Index 0 Param 1=&FloatFromCurve(FloatCurve=DataInterface() {propertyValues {Curve {keys [
         {arriveTangent=-0.1846, time=0, value=0}
         {arriveTangent=1.3385, interpMode=RCIM_Cubic, leaveTangent=1.1594, tangentMode=RCTM_Break, time=0.5, value=0}
         {arriveTangent=2, interpMode=RCIM_Cubic, leaveTangent=2, time=0.9, value=0.45}
         {arriveTangent=3.0091, interpMode=RCIM_Cubic, leaveTangent=3.0091, tangentMode=RCTM_User, time=1, value=1}
        ]}}})
    Index 0 Param 2=1
    Index 0 Param 3=1
    Index 0 Param 4=1
    Module.Index0Param0Write=true
    Module.Index0Param1Write=true
    Module.Index0Param2Write=true
    Module.Index0Param3Write=true
    Module.Index1Param0Write=false
    Module.Index1Param1Write=false
    Module.Index1Param2Write=false
    Module.Index1Param3Write=false
    Write Parameter Index 1=true
   )
   &ScaleColor(Scale RGB=&VectorFromFloat(Value=2), ScaleA=true, ScaleRGB=true)
  ]
  rendererClasses [&NiagaraSpriteRendererProperties]
  renderers [{rendererClass=&NiagaraSpriteRendererProperties}]
 }
 Fire001 {
  emitterUpdateScript [
   &EmitterState()
   &SpawnRate(Spawn Probability=$System.Timeline002.Float, SpawnRate=200, Use Spawn Probability=true) {module=SpawnRate001}
   &SpawnRate(SpawnRate=&Lerp_Float(A=20, B=40, Alpha=$System.Timeline002.Time), Use Spawn Probability=false) {enabled=false}
  ]
  particleSpawnScript [
   &InitializeParticle(
    Color=<rgba 0.7812 0.1217 0.1478 1>
    Lifetime=0.15
    Sprite Size=<2 15>
    Sprite Size Mode=::NewEnumerator4
    UsePositionOffset=false
    Write Material Random=true
   )
   &PartitionParticles(Partition Mode:ENiagara_PartitionMode=::NewEnumerator5)
   &SkeletalMeshLocation(
    Bone / Socket Index=$Particles.PartitionParticles.Partition
    MaskSpawnGroup=false
    Mesh Sampling Type=::NewEnumerator5
    Skeletal Mesh=$System.Skeletal Mesh
    Socket Sampling Mode=::NewEnumerator5
   )
   &AddVelocity(
    Cone Angle=25
    Cone Axis=&RotateVectorByQuaternion(Vector To Rotate=<1 0 0>, Quaternion=$Particles.SkeletalMeshLocation.SampledRotation)
    Module.UseConeFalloff=false
    Module.UseVelDistribution=false
    Override Seed=false
    Rotation Mode=::NewEnumerator5
    Velocity Mode=::NewEnumerator2
   )
  ]
  particleUpdateScript [
   &ParticleState()
   &ScaleSpriteSize(
    Scale Factor=&MakeVector2D(
     X=1
     Y=&FloatFromCurve(FloatCurve=DataInterface() {propertyValues {Curve {keys [
          {arriveTangent=1.5967, interpMode=RCIM_Cubic, leaveTangent=1.5967, tangentMode=RCTM_User, time=0, value=0}
          {arriveTangent=0.0111, interpMode=RCIM_Cubic, leaveTangent=0.0111, tangentMode=RCTM_User, time=1, value=1}
         ]}}})
    )
    Scale Sprite Size Mode=::NewEnumerator2
   )
   &SkeletalMeshLocation(
    Bone / Socket Index=$Particles.PartitionParticles.Partition
    Initialize Particle Position To Sampled Position=false
    MaskSpawnGroup=false
    Mesh Sampling Type=::NewEnumerator5
    Orientation Sampling=::NewEnumerator1
    Position Sampling=::NewEnumerator1
    Skeletal Mesh=$System.Skeletal Mesh
    Socket Sampling Mode=::NewEnumerator5
   ) {module=SkeletalMeshLocation001}
   &SolveForcesAndVelocity(Clamp Velocity=false, Limit Acceleration=false)
   &DynamicMaterialParameters(
    Index 0 Param 1=&FloatFromCurve(FloatCurve=DataInterface() {propertyValues {Curve {keys [
         {arriveTangent=-0.1846, time=0, value=0}
         {arriveTangent=1.3385, interpMode=RCIM_Cubic, leaveTangent=1.1594, tangentMode=RCTM_Break, time=0.5, value=0}
         {arriveTangent=2, interpMode=RCIM_Cubic, leaveTangent=2, time=0.9, value=0.45}
         {arriveTangent=3.0091, interpMode=RCIM_Cubic, leaveTangent=3.0091, tangentMode=RCTM_User, time=1, value=1}
        ]}}})
    Index 0 Param 2=1
    Index 0 Param 3=1
    Index 0 Param 4=1
    Module.Index0Param0Write=true
    Module.Index0Param1Write=true
    Module.Index0Param2Write=true
    Module.Index0Param3Write=true
    Module.Index1Param0Write=false
    Module.Index1Param1Write=false
    Module.Index1Param2Write=false
    Module.Index1Param3Write=false
    Write Parameter Index 1=true
   )
   &ScaleColor(
    Scale RGB=&VectorFromFloat(Value=&FloatFromCurve(
      FloatCurve=DataInterface() {propertyValues {Curve {keys [
          {interpMode=RCIM_Cubic, tangentMode=RCTM_User, time=0, value=1}
          {arriveTangent=-2.3883, interpMode=RCIM_Cubic, leaveTangent=-2.3883, tangentMode=RCTM_User, time=1, value=0}
         ]}}}
      Scale Curve=40
     ))
    ScaleA=true
    ScaleRGB=true
   )
  ]
  rendererClasses [&NiagaraSpriteRendererProperties]
  renderers [{rendererClass=&NiagaraSpriteRendererProperties}]
 }
 Glow {
  emitterUpdateScript [
   &EmitterState(Life Cycle Mode=::NewEnumerator1, Loop Behavior=::NewEnumerator1, Loop Duration Mode=::NewEnumerator1, UseLoopDelay=false)
   &SpawnRate(Spawn Probability=$System.Timeline002.Float, SpawnRate=40, Use Spawn Probability=true) {enabled=false}
   &SpawnBurst_Instantaneous(Spawn Count=2, Use Spawn Probability=false, UseLoopCountLimit=false)
  ]
  particleSpawnScript [
   &PartitionParticles()
   &InitializeParticle(
    Color=<rgba 1 0.3096 0.0383 1>
    Lifetime=0.35
    Sprite Size Mode=::NewEnumerator3
    Uniform Sprite Size=18
    UsePositionOffset=false
    Write Material Random=true
   )
   &CameraOffset(Camera Offset Amount=2)
  ]
  particleUpdateScript [
   &ParticleState(Kill Particles When Lifetime Has Elapsed=false, Let Infinitely Lived Particles Die When Emitter Deactivates=true)
   &SkeletalMeshLocation(
    Bone / Socket Index=$Particles.PartitionParticles.Partition
    MaskSpawnGroup=false
    Mesh Sampling Type=::NewEnumerator5
    Sampled Position Offset=<0 3 0>
    Skeletal Mesh=$System.Skeletal Mesh
    Socket Sampling Mode=::NewEnumerator5
   )
   &ScaleColor(
    Scale Alpha=&Multiply_Float(
     A=&Multiply_Float(
      A=&Lerp_Float(A=0, B=1.25, Alpha=$System.Timeline005.NormalizedTime)
      B=&FN_FrontEndFloatModifier(Gameplay=1, Front End=0.25)
     )
     B=&FloatFromCurve(
      FloatCurve=DataInterface() {propertyValues {Curve {keys [
          {arriveTangent=3.2046, interpMode=RCIM_Cubic, leaveTangent=3.2046, tangentMode=RCTM_User, time=0, value=0}
          {interpMode=RCIM_Cubic, tangentMode=RCTM_User, time=1, value=1}
         ]}}}
      CurveIndex=$System.Timeline002.Float
     )
    )
    ScaleA=true
    ScaleRGB=true
   )
   &DynamicMaterialParameters(
    Index 0 Param 1=3
    Index 0 Param 2=2
    Module.Index0Float0Write=true
    Module.Index0Float1Write=true
    Module.Index0Float2Write=true
    Module.Index0Float3Write=true
   )
  ]
  rendererClasses [&NiagaraSpriteRendererProperties]
  renderers [{rendererClass=&NiagaraSpriteRendererProperties}]
 }
 LightAndGlow {
  emitterUpdateScript [
   &EmitterState(
    Inactive Response=::NewEnumerator1
    Life Cycle Mode=::NewEnumerator1
    Loop Behavior=::NewEnumerator1
    Loop Duration Mode=::NewEnumerator1
    UseLoopDelay=false
   )
   &SpawnBurst_Instantaneous(Spawn Count=2, Use Spawn Probability=false, UseLoopCountLimit=false)
  ]
  particleSpawnScript [
   &InitializeParticle(Color=<rgba 1 0.1561 0 1>, Sprite Size Mode=::NewEnumerator3, Uniform Sprite Size=2.5, UsePositionOffset=false, Write Material Random=true)
   &SkeletalMeshLocation(
    Bone / Socket Index=&ReturnExecIndex()
    MaskSpawnGroup=false
    Mesh Sampling Type=::NewEnumerator5
    Sampled Position Offset=&SelectVectorFromArray(
     Direct Array Index=&ReturnExecIndex()
     Vector Selection Array=DataInterface() {propertyValues {FloatData [<0 0 0>, <0 0 0>]}}
    )
    Skeletal Mesh=$System.Skeletal Mesh
    Socket Sampling Mode=::NewEnumerator5
   )
   &CameraOffset()
  ]
  particleUpdateScript [
   &ParticleState(Kill Particles When Lifetime Has Elapsed=false, Let Infinitely Lived Particles Die When Emitter Deactivates=true)
   &SkeletalMeshLocation(
    Bone / Socket Index=&ReturnExecIndex()
    MaskSpawnGroup=false
    Mesh Sampling Type=::NewEnumerator5
    Sampled Position Offset=&SelectVectorFromArray(
     Direct Array Index=&ReturnExecIndex()
     Vector Selection Array=DataInterface() {propertyValues {FloatData [<0 0 0>, <0 0 0>]}}
    )
    Skeletal Mesh=$System.Skeletal Mesh
    Socket Sampling Mode=::NewEnumerator5
   ) {module=SkeletalMeshLocation001}
   &FN_Light_Attributes(
    Color=<rgba 1 0.1561 0 1>
    Create Color Attribute=true
    Create Exponent Attribute=true
    Create Position Attribute=true
    Create Radius Attribute=true
    Enable Particle Light Enable=true
    Exponent=1
    Light Enabled=true
    Position=&AddVectorToPosition(
     Position=$Particles.SkeletalMeshLocation001.SampledPosition
     Vector=&RotateVectorByQuaternion(
      Vector To Rotate=&SelectVectorFromArray(
       Direct Array Index=&ReturnExecIndex()
       Vector Selection Array=DataInterface() {propertyValues {FloatData [<3 0 0>, <3 0 -2>]}}
      )
      Quaternion=$Particles.SkeletalMeshLocation001.SampledRotation
     )
    )
    Radius=&SelectFloatFromArray(
     Direct Array Index=&ReturnExecIndex()
     Float Selection Array=DataInterface() {propertyValues {FloatData [4,4]}}
    )
    Scale Color=&Multiply_Float(
     A=&RemapRange(
      Input Value To Remap=&Multiply_Float(
       A=&FloatFromCurve(
        FloatCurve=DataInterface() {propertyValues {
          bUseLUT=false
          Curve {
           postInfinityExtrap=RCCE_Cycle
           preInfinityExtrap=RCCE_Cycle
           keys [
            <at 0 1 RCIM_Cubic>
            {arriveTangent=0.5283, interpMode=RCIM_Cubic, leaveTangent=0.5283, time=0.0746, value=0.875}
            {arriveTangent=-0.0227, interpMode=RCIM_Cubic, leaveTangent=-0.0227, tangentMode=RCTM_User, time=0.1682, value=1.0888}
            {arriveTangent=-1.3938, interpMode=RCIM_Cubic, leaveTangent=-1.3938, time=0.2019, value=0.6577}
            {arriveTangent=1.4729, interpMode=RCIM_Cubic, leaveTangent=1.4729, time=0.2748, value=0.9402}
            {arriveTangent=0.905, interpMode=RCIM_Cubic, leaveTangent=0.905, time=0.3464, value=0.8705}
            {arriveTangent=-1.2159, interpMode=RCIM_Cubic, leaveTangent=-1.2159, time=0.4074, value=1.0603}
            {arriveTangent=-1.5422, interpMode=RCIM_Cubic, leaveTangent=-1.5422, time=0.4365, value=0.7609}
            {arriveTangent=-0.8529, interpMode=RCIM_Cubic, leaveTangent=-0.8529, time=0.4871, value=0.9373}
            {arriveTangent=0.3027, interpMode=RCIM_Cubic, leaveTangent=0.3027, time=0.5331, value=0.6785}
            {arriveTangent=1.4034, interpMode=RCIM_Cubic, leaveTangent=1.4034, time=0.6112, value=0.9748}
            {arriveTangent=1.5913, interpMode=RCIM_Cubic, leaveTangent=1.5913, time=0.69, value=0.8986}
            {arriveTangent=-0.751, interpMode=RCIM_Cubic, leaveTangent=-0.751, time=0.7536, value=1.2014}
            {arriveTangent=-2.595, interpMode=RCIM_Cubic, leaveTangent=-2.595, time=0.7995, value=0.8164}
            {arriveTangent=-0.8671, interpMode=RCIM_Cubic, leaveTangent=-0.8671, time=0.8674, value=0.9059}
            {arriveTangent=-0.3671, interpMode=RCIM_Cubic, leaveTangent=-0.3671, time=0.9138, value=0.7173}
            {arriveTangent=-0.6714, interpMode=RCIM_Cubic, leaveTangent=-0.6714, time=0.9899, value=0.8609}
            {arriveTangent=0.7195, interpMode=RCIM_Cubic, leaveTangent=0.7195, time=1.0372, value=0.6345}
            {arriveTangent=1.5649, interpMode=RCIM_Cubic, leaveTangent=1.5649, time=1.099, value=0.9394}
            {arriveTangent=0.3682, interpMode=RCIM_Cubic, leaveTangent=0.3682, time=1.1296, value=0.7792}
            {arriveTangent=1.199, interpMode=RCIM_Cubic, leaveTangent=1.199, time=1.2078, value=0.9795}
            {arriveTangent=0.6999, interpMode=RCIM_Cubic, leaveTangent=0.6999, time=1.2281, value=0.8972}
            {arriveTangent=0.8013, interpMode=RCIM_Cubic, leaveTangent=0.8013, time=1.329, value=1.0643}
            {arriveTangent=0.4649, interpMode=RCIM_Cubic, leaveTangent=0.4649, time=1.3719, value=1.0125}
            {arriveTangent=-3.0415, interpMode=RCIM_Cubic, leaveTangent=-3.0415, time=1.414, value=1.1038}
            {arriveTangent=-0.0928, interpMode=RCIM_Cubic, leaveTangent=-0.0928, tangentMode=RCTM_User, time=1.4501, value=0.7746}
            {arriveTangent=0.7682, interpMode=RCIM_Cubic, leaveTangent=0.7682, time=1.5242, value=0.9899}
            {arriveTangent=0.5284, interpMode=RCIM_Cubic, leaveTangent=0.5284, time=1.5566, value=0.8565}
            {arriveTangent=-0.1056, interpMode=RCIM_Cubic, leaveTangent=-0.1056, time=1.7244, value=1.0957}
            {arriveTangent=-2.1193, interpMode=RCIM_Cubic, leaveTangent=-2.1193, time=1.7619, value=0.8348}
            {arriveTangent=-1.2854, interpMode=RCIM_Cubic, leaveTangent=-1.2854, time=1.7959, value=0.9442}
            {arriveTangent=-0.1279, interpMode=RCIM_Cubic, leaveTangent=-0.1279, time=1.836, value=0.7395}
            {arriveTangent=1.0187, interpMode=RCIM_Cubic, leaveTangent=1.0187, time=1.9343, value=0.9265}
            {arriveTangent=1.0782, interpMode=RCIM_Cubic, leaveTangent=1.0782, time=1.9912, value=0.8977}
            {arriveTangent=-2.5847, interpMode=RCIM_Cubic, leaveTangent=-2.5847, time=2.0488, value=1.0499}
            {arriveTangent=-0.6379, interpMode=RCIM_Cubic, leaveTangent=-0.6379, time=2.0828, value=0.6611}
            {arriveTangent=0.4353, interpMode=RCIM_Cubic, leaveTangent=0.4353, time=2.2158, value=0.9434}
            {arriveTangent=-0.3447, interpMode=RCIM_Cubic, leaveTangent=-0.3447, time=2.2673, value=0.7414}
            {arriveTangent=0.6603, interpMode=RCIM_Cubic, leaveTangent=0.6603, time=2.3258, value=0.9055}
            {arriveTangent=0.3801, interpMode=RCIM_Cubic, leaveTangent=0.3801, time=2.3971, value=0.8271}
            {arriveTangent=0.6687, interpMode=RCIM_Cubic, leaveTangent=0.6687, time=2.4654, value=0.9586}
            {arriveTangent=1.2527, interpMode=RCIM_Cubic, leaveTangent=1.2527, time=2.5143, value=0.9055}
            {arriveTangent=-0.515, interpMode=RCIM_Cubic, leaveTangent=-0.515, time=2.5566, value=1.0728}
            {arriveTangent=-1.6201, interpMode=RCIM_Cubic, leaveTangent=-1.6201, time=2.597, value=0.8629}
            {arriveTangent=-0.9669, interpMode=RCIM_Cubic, leaveTangent=-0.9669, time=2.6456, value=0.9286}
            {arriveTangent=0.1009, interpMode=RCIM_Cubic, leaveTangent=0.1009, time=2.6921, value=0.771}
            {arriveTangent=0.0929, interpMode=RCIM_Cubic, leaveTangent=0.0929, time=2.7509, value=0.9392}
            {arriveTangent=0.2939, interpMode=RCIM_Cubic, leaveTangent=0.2939, time=2.8654, value=0.7871}
            {arriveTangent=1.7227, interpMode=RCIM_Cubic, leaveTangent=1.7227, time=2.9271, value=0.991}
            {arriveTangent=0.1236, interpMode=RCIM_Cubic, leaveTangent=0.1236, time=2.957, value=0.9449}
            <at 3 1 RCIM_Cubic>
           ]
          }
         }}
        CurveIndex=$System.FlickerSpeed
       )
       B=1
      )
      Input Min=0.7
      Input Max=1.2
      Output Min=0
      Output Max=1
     )
     B=&Lerp_Float(A=10, B=20, Alpha=$System.Timeline002.NormalizedTime)
    )
   )
   &ScaleColor(
    Scale Alpha=&Multiply_Float(
     A=&RemapRange(
      Input Value To Remap=&Multiply_Float(
       A=&FloatFromCurve(
        FloatCurve=DataInterface() {propertyValues {
          bUseLUT=false
          Curve {
           postInfinityExtrap=RCCE_Cycle
           preInfinityExtrap=RCCE_Cycle
           keys [
            <at 0 1 RCIM_Cubic>
            {arriveTangent=0.5283, interpMode=RCIM_Cubic, leaveTangent=0.5283, time=0.0746, value=0.875}
            {arriveTangent=-0.0227, interpMode=RCIM_Cubic, leaveTangent=-0.0227, tangentMode=RCTM_User, time=0.1682, value=1.0888}
            {arriveTangent=-1.3938, interpMode=RCIM_Cubic, leaveTangent=-1.3938, time=0.2019, value=0.6577}
            {arriveTangent=1.4729, interpMode=RCIM_Cubic, leaveTangent=1.4729, time=0.2748, value=0.9402}
            {arriveTangent=0.905, interpMode=RCIM_Cubic, leaveTangent=0.905, time=0.3464, value=0.8705}
            {arriveTangent=-1.2159, interpMode=RCIM_Cubic, leaveTangent=-1.2159, time=0.4074, value=1.0603}
            {arriveTangent=-1.5422, interpMode=RCIM_Cubic, leaveTangent=-1.5422, time=0.4365, value=0.7609}
            {arriveTangent=-0.8529, interpMode=RCIM_Cubic, leaveTangent=-0.8529, time=0.4871, value=0.9373}
            {arriveTangent=0.3027, interpMode=RCIM_Cubic, leaveTangent=0.3027, time=0.5331, value=0.6785}
            {arriveTangent=1.4034, interpMode=RCIM_Cubic, leaveTangent=1.4034, time=0.6112, value=0.9748}
            {arriveTangent=1.5913, interpMode=RCIM_Cubic, leaveTangent=1.5913, time=0.69, value=0.8986}
            {arriveTangent=-0.751, interpMode=RCIM_Cubic, leaveTangent=-0.751, time=0.7536, value=1.2014}
            {arriveTangent=-2.595, interpMode=RCIM_Cubic, leaveTangent=-2.595, time=0.7995, value=0.8164}
            {arriveTangent=-0.8671, interpMode=RCIM_Cubic, leaveTangent=-0.8671, time=0.8674, value=0.9059}
            {arriveTangent=-0.3671, interpMode=RCIM_Cubic, leaveTangent=-0.3671, time=0.9138, value=0.7173}
            {arriveTangent=-0.6714, interpMode=RCIM_Cubic, leaveTangent=-0.6714, time=0.9899, value=0.8609}
            {arriveTangent=0.7195, interpMode=RCIM_Cubic, leaveTangent=0.7195, time=1.0372, value=0.6345}
            {arriveTangent=1.5649, interpMode=RCIM_Cubic, leaveTangent=1.5649, time=1.099, value=0.9394}
            {arriveTangent=0.3682, interpMode=RCIM_Cubic, leaveTangent=0.3682, time=1.1296, value=0.7792}
            {arriveTangent=1.199, interpMode=RCIM_Cubic, leaveTangent=1.199, time=1.2078, value=0.9795}
            {arriveTangent=0.6999, interpMode=RCIM_Cubic, leaveTangent=0.6999, time=1.2281, value=0.8972}
            {arriveTangent=0.8013, interpMode=RCIM_Cubic, leaveTangent=0.8013, time=1.329, value=1.0643}
            {arriveTangent=0.4649, interpMode=RCIM_Cubic, leaveTangent=0.4649, time=1.3719, value=1.0125}
            {arriveTangent=-3.0415, interpMode=RCIM_Cubic, leaveTangent=-3.0415, time=1.414, value=1.1038}
            {arriveTangent=-0.0928, interpMode=RCIM_Cubic, leaveTangent=-0.0928, tangentMode=RCTM_User, time=1.4501, value=0.7746}
            {arriveTangent=0.7682, interpMode=RCIM_Cubic, leaveTangent=0.7682, time=1.5242, value=0.9899}
            {arriveTangent=0.5284, interpMode=RCIM_Cubic, leaveTangent=0.5284, time=1.5566, value=0.8565}
            {arriveTangent=-0.1056, interpMode=RCIM_Cubic, leaveTangent=-0.1056, time=1.7244, value=1.0957}
            {arriveTangent=-2.1193, interpMode=RCIM_Cubic, leaveTangent=-2.1193, time=1.7619, value=0.8348}
            {arriveTangent=-1.2854, interpMode=RCIM_Cubic, leaveTangent=-1.2854, time=1.7959, value=0.9442}
            {arriveTangent=-0.1279, interpMode=RCIM_Cubic, leaveTangent=-0.1279, time=1.836, value=0.7395}
            {arriveTangent=1.0187, interpMode=RCIM_Cubic, leaveTangent=1.0187, time=1.9343, value=0.9265}
            {arriveTangent=1.0782, interpMode=RCIM_Cubic, leaveTangent=1.0782, time=1.9912, value=0.8977}
            {arriveTangent=-2.5847, interpMode=RCIM_Cubic, leaveTangent=-2.5847, time=2.0488, value=1.0499}
            {arriveTangent=-0.6379, interpMode=RCIM_Cubic, leaveTangent=-0.6379, time=2.0828, value=0.6611}
            {arriveTangent=0.4353, interpMode=RCIM_Cubic, leaveTangent=0.4353, time=2.2158, value=0.9434}
            {arriveTangent=-0.3447, interpMode=RCIM_Cubic, leaveTangent=-0.3447, time=2.2673, value=0.7414}
            {arriveTangent=0.6603, interpMode=RCIM_Cubic, leaveTangent=0.6603, time=2.3258, value=0.9055}
            {arriveTangent=0.3801, interpMode=RCIM_Cubic, leaveTangent=0.3801, time=2.3971, value=0.8271}
            {arriveTangent=0.6687, interpMode=RCIM_Cubic, leaveTangent=0.6687, time=2.4654, value=0.9586}
            {arriveTangent=1.2527, interpMode=RCIM_Cubic, leaveTangent=1.2527, time=2.5143, value=0.9055}
            {arriveTangent=-0.515, interpMode=RCIM_Cubic, leaveTangent=-0.515, time=2.5566, value=1.0728}
            {arriveTangent=-1.6201, interpMode=RCIM_Cubic, leaveTangent=-1.6201, time=2.597, value=0.8629}
            {arriveTangent=-0.9669, interpMode=RCIM_Cubic, leaveTangent=-0.9669, time=2.6456, value=0.9286}
            {arriveTangent=0.1009, interpMode=RCIM_Cubic, leaveTangent=0.1009, time=2.6921, value=0.771}
            {arriveTangent=0.0929, interpMode=RCIM_Cubic, leaveTangent=0.0929, time=2.7509, value=0.9392}
            {arriveTangent=0.2939, interpMode=RCIM_Cubic, leaveTangent=0.2939, time=2.8654, value=0.7871}
            {arriveTangent=1.7227, interpMode=RCIM_Cubic, leaveTangent=1.7227, time=2.9271, value=0.991}
            {arriveTangent=0.1236, interpMode=RCIM_Cubic, leaveTangent=0.1236, time=2.957, value=0.9449}
            <at 3 1 RCIM_Cubic>
           ]
          }
         }}
        CurveIndex=$System.FlickerSpeed
       )
       B=1
      )
      Input Min=0.6
      Input Max=1.2
      Output Min=0
      Output Max=1
     )
     B=&Multiply_Float(
      A=&Lerp_Float(
       A=&FN_FrontEndFloatModifier(Gameplay=0.1, Front End=0.05)
       B=&FN_FrontEndFloatModifier(Gameplay=0.25, Front End=0.1)
       Alpha=$System.Timeline005.NormalizedTime
      )
      B=1
     )
    )
    Scale RGB=&VectorFromFloat(Value=&FN_FrontEndFloatModifier(Gameplay=1.5, Front End=1))
    ScaleA=true
    ScaleRGB=true
   )
   &DynamicMaterialParameters(
    Index 0 Param 1=2
    Index 0 Param 2=1
    Module.Index0Float0Write=true
    Module.Index0Float1Write=true
    Module.Index0Float2Write=true
    Module.Index0Float3Write=true
   )
   /SetVariables(
    Particles.AirGlowCameraOffset=2
    Particles.AirGlowColor:LinearColor=&MultiplyLinearColor(
     Linear Color=$Particles.Color
     Scale Factor=&MakeFloat4FromVectorAndFloat(
      "Vector (RGB)"=&VectorFromFloat(Value=1)
      "Float (Alpha)"=&Lerp_Float(A=0.5, B=2, Alpha=$System.Timeline005.NormalizedTime)
     )
    )
    Particles.AirGlowPosition:NiagaraPosition=&AddVectorToPosition(
     Position=$Particles.SkeletalMeshLocation001.SampledPosition
     Vector=&RotateVectorByQuaternion(Vector To Rotate=<3 0 0>, Quaternion=$Particles.SkeletalMeshLocation001.SampledRotation)
    )
    Particles.AirGlowSize:Vector2f=&Multiply_Vector2D_ByFloat(Float=3, Vector2D=$Particles.SpriteSize)
   ) {isSetParameters=true}
   &SpriteFacingAndAlignment(
    Sprite Facing=&RotateVectorByQuaternion(Vector To Rotate=<1 0 0>, Quaternion=$Particles.SkeletalMeshLocation001.SampledRotation)
    WriteAlignment=false
    WriteFacing=true
   )
  ]
  rendererClasses [&NiagaraLightRendererProperties, &NiagaraSpriteRendererProperties]
  renderers [
   {rendererClass=&NiagaraLightRendererProperties, rendererIndex=0}
   {rendererClass=&NiagaraSpriteRendererProperties, rendererIndex=1}
   {rendererClass=&NiagaraSpriteRendererProperties, rendererIndex=2}
  ]
 }
 Minimal {
  emitterUpdateScript [
   &EmitterState(Life Cycle Mode=::NewEnumerator1, Loop Behavior=::NewEnumerator1, Loop Duration Mode=::NewEnumerator1, UseLoopDelay=false)
   &SpawnBurst_Instantaneous(Use Spawn Probability=false, UseLoopCountLimit=false)
  ]
  particleSpawnScript [
   &InitializeParticle(
    Color=<rgba 1 0.056 0.0002 1>
    Sprite Size Mode=::NewEnumerator3
    Uniform Sprite Size=13
    UsePositionOffset=false
    Write Material Random=true
   )
   &CameraOffset(Camera Offset Amount=2)
  ]
  particleUpdateScript [
   &ParticleState(Kill Particles When Lifetime Has Elapsed=false, Let Infinitely Lived Particles Die When Emitter Deactivates=true)
   &SkeletalMeshLocation(Bone Sampling Mode=::NewEnumerator4, MaskSpawnGroup=false, Sampled Position Offset=<0 -7.5 0>, Skeletal Mesh=$System.Skeletal Mesh)
   &ScaleColor(
    Scale Alpha=&Multiply_Float(
     A=&Lerp_Float(A=0.25, B=0.6, Alpha=$System.Timeline005.NormalizedTime)
     B=&FN_FrontEndFloatModifier(Gameplay=1, Front End=0.25)
    )
    ScaleA=true
    ScaleRGB=true
   )
   &DynamicMaterialParameters(
    Index 0 Param 1=3
    Index 0 Param 2=2
    Module.Index0Float0Write=true
    Module.Index0Float1Write=true
    Module.Index0Float2Write=true
    Module.Index0Float3Write=true
   )
  ]
  rendererClasses [&NiagaraSpriteRendererProperties]
  renderers [{rendererClass=&NiagaraSpriteRendererProperties}]
 }
 NE_FN_Sparks {
  simTarget=GPUComputeSim
  emitterUpdateScript [
   &EmitterState()
   &SpawnRate(Spawn Probability=$System.Timeline002.Float, SpawnRate=400, Use Spawn Probability=true)
  ]
  particleSpawnScript [
   /SetVariables(Particles.SparkIntensity:NiagaraFloat=&FloatFromCurve(
     FloatCurve=DataInterface() {propertyValues {Curve {keys [
         <at 0 1 RCIM_Cubic>
         {arriveTangent=-2.5317, interpMode=RCIM_Cubic, leaveTangent=-2.5317, tangentMode=RCTM_User, time=0.6, value=0.5}
         <at 1 0 RCIM_Cubic>
        ]}}}
     CurveIndex=&RandomRangeFloat(Minimum=0, Maximum=1)
    )) {isSetParameters=true}
   &PartitionParticles()
   &InitializeParticle(
    Color=&LerpLinearColors(StartColor=<rgba 10 1.5225 0.5208 1>, EndColor=<rgba 25 16.8856 2.9948 1>, LerpFactor=$Particles.SparkIntensity)
    Lifetime Max=0.5
    Lifetime Min=0.15
    Lifetime Mode=::NewEnumerator1
    Mass=&Lerp_Float(A=0.3, B=1, Alpha=$Particles.SparkIntensity)
    Mass Mode=::NewEnumerator1
    Sprite Size Mode=::NewEnumerator3
    Uniform Sprite Size=&Lerp_Float(A=0.3, B=0.4, Alpha=$Particles.SparkIntensity)
    UsePositionOffset=false
    Write Material Random=true
   )
   /SetVariables(Particles.SpriteSize=&Multiply_Vector2(
     A=$Particles.SpriteSize
     B=&Vector2DFromFloat(Value=&FloatFromCurve(FloatCurve=DataInterface() {propertyValues {Curve {keys [<at 1000 1 RCIM_Cubic>,<at 5000 2>]}}}, CurveIndex=$Engine.Owner.LODDistance))
    )) {isSetParameters=true}
   &SkeletalMeshLocation(
    Bone / Socket Index=$Particles.PartitionParticles.Partition
    MaskSpawnGroup=false
    Mesh Sampling Type=::NewEnumerator5
    Skeletal Mesh=$System.Skeletal Mesh
    Socket Sampling Mode=::NewEnumerator5
   )
   &SphereLocation(
    MaskSpawnGroup=false
    Override Seed=false
    Sphere Radius=1
    Surface Only=false
    UseNonUniformScale=false
    UseOffset=false
    UseSphereOrientation=false
   )
   &AddVelocityInCone(
    Cone Angle=20
    Cone Axis=&RotateVectorByQuaternion(Vector To Rotate=<1 0 0>, Quaternion=$Particles.SkeletalMeshLocation.SampledRotation)
    Cone Axis Coordinate Space=::World
    Module.Use Velocity Falloff On Cone Axis=false
    Velocity Strength=&RandomRangeFloat(Minimum=100, Maximum=170)
   )
  ]
  particleUpdateScript [
   &ParticleState()
   &SkeletalMeshLocation(
    Bone / Socket Index=$Particles.PartitionParticles.Partition
    Initialize Particle Position To Sampled Position=false
    MaskSpawnGroup=false
    Mesh Sampling Type=::NewEnumerator5
    Orientation Sampling=::NewEnumerator1
    Position Sampling=::NewEnumerator1
    Skeletal Mesh=$System.Skeletal Mesh
    Socket Sampling Mode=::NewEnumerator5
   ) {module=SkeletalMeshLocation001}
   &GravityForce(Gravity=<0 0 -450>)
   &Drag(Use Linear Drag=true, Use Rotational Drag=false)
   &VectorNoiseForce(
    Force Amount=&FloatFromCurve(
     FloatCurve=DataInterface() {propertyValues {Curve {keys [
         {arriveTangent=-9.4515, interpMode=RCIM_Cubic, leaveTangent=-4.1687, tangentMode=RCTM_User, time=0, value=1}
         {arriveTangent=-3.4372, interpMode=RCIM_Cubic, leaveTangent=-3.4372, tangentMode=RCTM_User, time=0.6, value=0.5}
        ]}}}
     Scale Curve=600
    )
    Module.Use Random Noise Time Interval=false
   )
   &SolveForcesAndVelocity(Clamp Velocity=false, Limit Acceleration=false)
   &ScaleColor(
    Linear Color Curve=DataInterface() {propertyValues {
      AlphaCurve {keys [<at 0 1>,<at 1 1>]}
      BlueCurve {keys [
        <at 0 2 RCIM_Cubic>
        {arriveTangent=-1.452, interpMode=RCIM_Cubic, leaveTangent=-1.452, tangentMode=RCTM_User, time=0.1467, value=1}
        <at 1 0.0755>
       ]}
      GreenCurve {keys [
        <at 0 2 RCIM_Cubic>
        {arriveTangent=-1.4408, interpMode=RCIM_Cubic, leaveTangent=-1.4408, tangentMode=RCTM_User, time=0.1467, value=1}
        <at 1 0.078>
       ]}
      RedCurve {keys [
        <at 0 2 RCIM_Cubic>
        {arriveTangent=-0.8332, interpMode=RCIM_Cubic, leaveTangent=-0.8332, tangentMode=RCTM_User, time=0.1467, value=1}
        <at 1 0.25>
       ]}
     }}
    Scale Mode=::NewEnumerator2
   )
   &ScaleSpriteSize(Scale Factor=&Vector2DFromFloat(Value=&FloatFromCurve(FloatCurve=DataInterface() {propertyValues {Curve {keys [<at 0.4 1 RCIM_Cubic>, {arriveTangent=-3.8662, interpMode=RCIM_Cubic, leaveTangent=-3.8662, tangentMode=RCTM_User, time=1, value=0}]}}})))
   &ScaleSpriteSizeBySpeed(
    Max Scale Factor=<1 5>
    Sample Scale Factor By Curve=true
    Scale Factor Curve=DataInterface() {propertyValues {Curve {keys [
        {arriveTangent=0.2757, time=0, value=0}
        {arriveTangent=0.3844, interpMode=RCIM_Cubic, leaveTangent=0.3844, tangentMode=RCTM_User, time=0.3, value=0.15}
        {arriveTangent=2.2765, interpMode=RCIM_Cubic, leaveTangent=2.2765, tangentMode=RCTM_User, time=1, value=1}
       ]}}}
    Velocity Threshold=50
   )
  ]
  rendererClasses [&NiagaraSpriteRendererProperties]
  renderers [{rendererClass=&NiagaraSpriteRendererProperties}]
 }
}
summary {
 systemName=NS_Backpack_BadCat
 emitters [
  {emitterName=FN_Core_Fiery_Smoke_Sprite, rendererClasses [&NiagaraSpriteRendererProperties]}
  {emitterName=LightAndGlow, rendererClasses [&NiagaraLightRendererProperties, &NiagaraSpriteRendererProperties]}
  {emitterName=Decal, rendererClasses [&NiagaraSpriteRendererProperties]}
  {emitterName=Fire, rendererClasses [&NiagaraSpriteRendererProperties]}
  {emitterName=NE_FN_Sparks, simTarget=GPUComputeSim, rendererClasses [&NiagaraSpriteRendererProperties]}
  {emitterName=Fire001, rendererClasses [&NiagaraSpriteRendererProperties]}
  {emitterName=FN_Core_Fiery_Smoke_Sprite001, rendererClasses [&NiagaraSpriteRendererProperties]}
  {emitterName=Minimal, rendererClasses [&NiagaraSpriteRendererProperties]}
  {emitterName=Glow, rendererClasses [&NiagaraSpriteRendererProperties]}
 ]
 userVariables [
  User.bDebug {type=NiagaraBool}
  User.bIsFrontEnd {type=NiagaraBool}
  User.bIsFrontEndPreview {type=NiagaraBool}
  User.PlayerJumping {type=NiagaraBool}
  User.SprintReactive {type=NiagaraBool}
 ]
}
