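# Parent material for the BadCat backpack that makes its emissive glow respond to Blueprint-fed motion and timing.
# It wraps the shared 2023 character PreFX and PostFX functions, scaling Emissive Color between them.
# Glow strength follows accumulated velocity plus a rerouted value, with a lower peak in the front end.
# A Timestamp_FrontEnd pulse, timed from a Blueprint timestamp, ramps up, holds about two seconds, then fades.
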Material /BRCosmetics/Accessories/FORT_Backpacks/Backpack_M_MED_BadCat/Materials/M_BadCat_Backpack_PARENT
graph {
 properties {
  bUseMaterialAttributes=true
  bUsedWithMeshDeformer=true
  bUsedWithMorphTargets=true
  bUsedWithSkeletalMesh=true
  pixelDepthOffsetMode=PDOM_Legacy
  shadingModel=MSM_FromMaterialExpression
  usedShadingModels=Default Lit | Subsurface | Clear Coat | Subsurface Profile | Cloth
 }
 nodes [
  n1 = MaterialFunctionCall() {materialFunction=&MF_CharacterParent_2023_PreFX}
  n2 = GetMaterialAttributes(None=@n1.Result) {attributeGetTypes [#1]}
  n3 = LinearInterpolate(
   B=LinearInterpolate(Alpha=ScalarParameter() {parameterName=BP_IsFrontEnd}) {constA=8.6, constB=2}
   Alpha=Add(A=ScalarParameter() {parameterName=BP_AccumulatedVelocity}, B=NamedRerouteUsage())
  ) {constB=8}
  MaterialFunctionCall(
   Material=SetMaterialAttributes(MaterialAttributes=@n2.MaterialAttributes, Emissive Color=Multiply(A=@n2.Emissive Color, B=@n3)) {attributeSetTypes [#1]}
   M=@n1.M
   Pre Skinned Local Position=@n1.M
   Distance Blend=@n1.M
   Fresnel=@n1.M
  ) {materialFunction=&MF_CharacterParent_2023_PostFX}
  n4 = Subtract(A=Time(), B=ScalarParameter() {defaultValue=100000000376832, parameterName=BP_GetTime})
  n5 = Multiply(A=@n4) {constB=3}
  n6 = Saturate(None=Multiply(A=Subtract(A=@n4) {constB=2}) {constB=3})
  NamedRerouteDeclaration(None=Multiply(A=Saturate(None=@n5), B=OneMinus(None=@n6))) {name=Timestamp_FrontEnd, nodeColor=<rgba 1 0 0.9609 1>}
 ]
}
