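# Animation Blueprint for the Bad Cat backpack that spins its tire and plays a jitter animation, both responsive to the owner's movement.
# Each update, TireRotate advances a wrapping spin angle using an interpolated acceleration that rises while moving, sliding or jumping, then drives the tire_dyn bone's roll.
# An additive jitter sequence, slower during movement, layers over the offset pack bone.
# The jitter_dyn bone node and sine wobble logic are unconnected.
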AnimBlueprint /BRCosmetics/Accessories/FORT_Backpacks/Backpack_M_MED_BadCat/Meshes/M_MED_BadCat_Pack_AnimBP
target_skeleton=&/M_MED_BadCat_Pack_Skeleton
graphs [
 AnimGraph {nodes [
   "Transform (Modify) Bone - Bone: jitter_dyn"(Alpha=1, Translation="0,0,0", Rotation="0,0,0", Scale="1,1,1") {translationMode=BMM_Additive, translationSpace=BCS_BoneSpace, alphaBoolBlend {blendOption=Linear}, boneToModify {boneName=jitter_dyn}}
   n1 = "Transform (Modify) Bone - Bone: pack"(ComponentPose=Local To Component(LocalPose=Input Pose()), Translation="0,-2,0", Rotation="0,0,0") {
    translationMode=BMM_Additive
    translationSpace=BCS_BoneSpace
    alphaBoolBlend {blendOption=Linear}
    boneToModify {boneName=pack}
    translation {y=-2}
   }
   n2 = "Transform (Modify) Bone - Bone: tire_dyn"(ComponentPose=@n1, Alpha=1, Translation="0,0,0", Rotation=Get Spin result() -> (Spin result:Rotator), Scale="1,1,1") {rotationMode=BMM_Additive, rotationSpace=BCS_BoneSpace, alphaBoolBlend {blendOption=Linear}, boneToModify {boneName=tire_dyn}}
   n3 = Get Jitter Rate() -> (Jitter Rate:double)
   Output Pose(Result=Apply Additive(
     Base=Component To Local(ComponentPose=@n2)
     Additive="Sequence Player (additive) 'BadCat_JItter_Anim_PACK'"!(PlayRateBasis:float=@n3) -> (Pose) {sequence=/BRCosmetics/Accessories/FORT_Backpacks/Backpack_M_MED_BadCat/Anim/BadCat_JItter_Anim_PACK}
    ) {alphaBoolBlend {blendOption=Linear}}) {layerGroup=None}
  ]}
 EventGraph {nodes [
   n1 = Event BlueprintUpdateAnimation()
   > Set Delta Time X(Delta Time X=@n1.DeltaTimeX)
   > n2 = TireRotate!(execute, self, Spin:double=Get Spin() -> (Spin:double), Speed:double=Get Speed() -> (Speed:double)) -> (then, Output:double)
   > n3 = Set Spin(Spin:double=@n2) -> (Output_Get:double)
   n4 = Get bIsJumping()
   n5 = Get bIsSliding()
   n6 = Get bIsMoving2D()
   > Set Jitter Rate(Jitter Rate:double=SelectFloat(A=0.7, B=1, bPickA=OR Boolean!(A=@n6, B=@n5, C:bool=@n4) -> (ReturnValue))) -> (Output_Get:double)
   > Set Spin result(Spin result:Rotator=Invert Rotator!(A:Rotator=MakeRotator(Roll=@n3)) -> (ReturnValue:Rotator)) -> (Output_Get:Rotator)
   TryGetPawnOwner()
   n7 = GetRealTimeSeconds()
   n8 = "Sin (Radians)"(A=float * float(A=float * float(A=7, B=7), B=@n7))
   n9 = "Sin (Radians)"(A=float * float(A=float * float(A=7, B=7), B=@n7))
   n10 = "Sin (Radians)"(A=float * float(A=float * float(A=7, B=7), B=@n7))
   Set dyn_pack_Float(dyn_pack_Float:Vector=MakeVector(X=float * float(A=@n9, B=0.06), Y=float * float(A=@n8, B=0.06), Z=float * float(A=@n10, B=0.06))) -> (Output_Get:Vector)
  ]}
 TireRotate {nodes [
   n1 = TireRotate!() -> (then, Spin:double=~, Speed:double=~) {is=title K2Node_FunctionEntry}
   > n2 = Set Spin var(Spin var:double=@n1) -> (Output_Get:double)
   n3 = OR Boolean(A=Get bIsMoving2D(), B=Get bIsSliding(), C=Get bIsJumping())
   > Set Acceleration(Acceleration:double=FInterpTo(
     Current=Get Acceleration() -> (Acceleration:double)
     Target=SelectFloat(A=1.79, B=0.5, bPickA=@n3)
     DeltaTime=Get Delta Time X()
     InterpSpeed=3
    )) -> (Output_Get:double)
   > n4 = Branch(Condition="float >= float"(A=@n2, B=720))
   > n5 = Set Spin var(Spin var=float - float(A=Get Spin var() -> (Spin var:double), B=720))
   n6 = Get Velocity mult() -> (Velocity mult:double)
   n7 = Get Acceleration()
   n8 = Get Delta time remap() -> (Delta time remap:double)
   n9 = Set Spin var(
    execute=@n4.else
    Spin var=float + float!(A=Get Spin var(), B=float * float(A=@n1.Speed, B=@n8) {is=title K2Node_PromotableOperator}, C:double=float * float(A=@n7, B=@n6)) -> (ReturnValue)
   )
   Return Node!(execute=[@n9.then, @n5.then], Output:double=@n9)
  ]}
]
variables [
 Acceleration {type=number}
 Backpack offset {type=boolean}
 Delta time remap {type=number}
 Delta Time X {type=number}
 dyn_pack_Float {type=Vector}
 Jitter Rate {type=number}
 Speed {type=number}
 Spin {type=number}
 Spin result {type=Rotator}
 Spin var {type=number}
 Velocity mult {type=number, default=12}
]
